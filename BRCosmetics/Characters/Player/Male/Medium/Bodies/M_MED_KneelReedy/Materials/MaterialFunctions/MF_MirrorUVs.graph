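# Material function in the KneelReedy character body's folder that returns UVs mirrored in U, V or both, exposed under the DelMarUI library category.
# Each axis has a static switch between the raw coordinate and one folded around 0.5, with U mirrored by default.
# U also gets a Mirror Offset, then a Step-driven lerp flips the result on one side of center, even with MirrorU off.
# V has no offset, and both axes are appended for output.
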MaterialFunction /BRCosmetics/Characters/Player/Male/Medium/Bodies/M_MED_KneelReedy/Materials/MaterialFunctions/MF_MirrorUVs
graph {
 properties {bExposeToLibrary=true, description="Returns mirrored UVs (mirrored in U, V, or both).", libraryCategoriesText [DelMarUI]}
 nodes [
  n1 = FunctionInput(Preview=TextureCoordinate()) {
   bUsePreviewValueAsDefault=true
   description="The input UVs to mirror (uses TexCoord 0 by default)."
   inputName=UVs
   inputType=FunctionInput_Vector2
  }
  n2 = ComponentMask(None=@n1) {r=true}
  n3 = ConstantBiasScale(None=@n2) {bias=-0.5, scale=2}
  n4 = FunctionInput(Preview=StaticBool() {value=true}) {
   bUsePreviewValueAsDefault=true
   description="Whether to mirror in U (horizontally)."
   inputName=MirrorU
   inputType=FunctionInput_StaticBool
   sortPriority=1
  }
  n5 = StaticSwitch(True=OneMinus(None=Abs(None=@n3)), False=@n2, Value=@n4)
  n6 = FunctionInput(Preview=ScalarParameter() {defaultValue=-0.248, parameterName=Mirror Offset}) {bUsePreviewValueAsDefault=true, inputName=Mirror Offset, inputType=FunctionInput_Scalar}
  n7 = LinearInterpolate(A=OneMinus(None=Add(A=@n5, B=@n6)), B=Add(A=@n6, B=@n5), Alpha=Step(Y=@n3) {constX=0})
  n8 = ComponentMask(None=@n1) {g=true}
  n9 = OneMinus(None=Abs(None=ConstantBiasScale(None=@n8) {bias=-0.5, scale=2}))
  n10 = FunctionInput(Preview=StaticBool()) {
   bUsePreviewValueAsDefault=true
   description="Whether to mirror in V (vertically)."
   inputName=MirrorV
   inputType=FunctionInput_StaticBool
   sortPriority=2
  }
  FunctionOutput(None=AppendVector(A=OneMinus(None=@n7), B=StaticSwitch(True=@n9, False=@n8, Value=@n10))) {outputName=MirroredUVs}
 ]
}
