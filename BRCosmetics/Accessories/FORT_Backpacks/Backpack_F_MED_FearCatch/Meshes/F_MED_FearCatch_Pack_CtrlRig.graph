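# Control Rig for the FearCatch backpack cosmetic that adjusts its two eye bones, dyn_eye_l and dyn_eye_r.
# On forward solve, each eye is parent-constrained to its matching dyn_eyeAim bone with offset maintained, filtered to Y and Z translation only.
# Small X translation offsets then nudge each eye, slightly more on the left than the right.
# The leg bones are in the hierarchy but untouched by the graph.
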ControlRigBlueprint /BRCosmetics/Accessories/FORT_Backpacks/Backpack_F_MED_FearCatch/Meshes/F_MED_FearCatch_Pack_CtrlRig
graphs [
 RigVMFunctionLibrary {links [], nodes []}
 RigVMModel {nodes [
   RigUnit_BeginExecution = &RigUnit_BeginExecution()
   &RigUnit_ParentConstraint(
    AdvancedSettings=(InterpolationType=Average,RotationOrderForFilter=XYZ)
    Child=(Type=Bone,Name="dyn_eye_l")
    Filter=(TranslationFilter=(bX=false,bY=true,bZ=true),RotationFilter=(bX=false,bY=false,bZ=false),ScaleFilter=(bX=false,bY=false,bZ=false))
    Parents=((Item=(Type=Bone,Name="dyn_eyeAim_l"),Weight=1))
    Weight=1
    bMaintainOffset=true
    ExecutePin=@RigUnit_BeginExecution.ExecutePin
   )
   > &RigUnit_ParentConstraint(
    AdvancedSettings=(InterpolationType=Average,RotationOrderForFilter=XZY)
    Child=(Type=Bone,Name="dyn_eye_r")
    Filter=(TranslationFilter=(bX=false,bY=true,bZ=true),RotationFilter=(bX=false,bY=false,bZ=false),ScaleFilter=(bX=false,bY=false,bZ=false))
    Parents=((Item=(Type=Bone,Name="dyn_eyeAim_r"),Weight=1))
    Weight=1
    bMaintainOffset=true
   )
   > &RigUnit_OffsetTransformForItem(
    Item=(Type=Bone,Name="dyn_eye_l")
    OffsetTransform=(Rotation=(X=0,Y=0,Z=0,W=1),Translation=(X=0.2,Y=0,Z=0),Scale3D=(X=1,Y=1,Z=1))
    Weight=1
    bPropagateToChildren=True
   )
   > &RigUnit_OffsetTransformForItem(
    Item=(Type=Bone,Name="dyn_eye_r")
    OffsetTransform=(Rotation=(X=0,Y=0,Z=0,W=1),Translation=(X=0.1,Y=0,Z=0),Scale3D=(X=1,Y=1,Z=1))
    Weight=1
    bPropagateToChildren=True
   )
  ]}
]
hierarchy [
 root {loc=<0 0 0>, rot=<0 0 0 1>}
 pack {parent=root, loc=<0 -0.7316 0>, rot=<0 0 0 1>}
 base_dyn_leg_r_00 {parent=pack, loc=<7.4296 -3.5796 8.7147>, rot=<0.0465 0.2501 -0.1767 -0.9508>}
 dyn_leg_r_00 {parent=base_dyn_leg_r_00, loc=<0 0 0>, rot=<0 0 0 1>}
 base_dyn_leg_r_01 {parent=pack, loc=<8.6958 -3.2716 5.3792>, rot=<0.0022 0.0146 -0.1507 -0.9885>}
 dyn_leg_r_01 {parent=base_dyn_leg_r_01, loc=<0 0 0>, rot=<0 0 0 1>}
 base_dyn_leg_r_02 {parent=pack, loc=<8.6711 -3.4929 0.4961>, rot=<0 -0.7071 0 -0.7071>}
 dyn_leg_r_02 {parent=base_dyn_leg_r_02, loc=<0 0 0>, rot=<0 0 0 1>}
 base_dyn_leg_r_03 {parent=pack, loc=<8.0386 -3.688 -2.1611>, rot=<0 -0.7071 0 -0.7071>}
 dyn_leg_r_03 {parent=base_dyn_leg_r_03, loc=<0 0 0>, rot=<0 0 0 1>}
 base_dyn_leg_l_00 {parent=pack, loc=<-7.43 -3.58 8.715>, rot=<-0.1767 -0.9508 0.0465 0.2501>}
 dyn_leg_l_00 {parent=base_dyn_leg_l_00, loc=<0 0 0>, rot=<0 0 0 1>}
 base_dyn_leg_l_01 {parent=pack, loc=<-8.696 -3.272 5.379>, rot=<-0.1508 -0.9885 0.0022 0.0146>}
 dyn_leg_l_01 {parent=base_dyn_leg_l_01, loc=<0 0 0>, rot=<0 0 0 1>}
 base_dyn_leg_l_02 {parent=pack, loc=<-8.671 -3.493 0.496>, rot=<0 -0.7071 0 -0.7071>}
 dyn_leg_l_02 {parent=base_dyn_leg_l_02, loc=<0 0 0>, rot=<0 0 0 1>}
 base_dyn_leg_l_03 {parent=pack, loc=<-8.0386 -3.688 -2.161>, rot=<0 -0.7071 0 -0.7071>}
 dyn_leg_l_03 {parent=base_dyn_leg_l_03, loc=<0 0 0>, rot=<0 0 0 1>}
 base_dyn_eyeAim_r {parent=pack, loc=<3.116 -7.6559 8.5717>, rot=<-0.4778 0.4991 0.4041 -0.5994>}
 dyn_eyeAim_r {parent=base_dyn_eyeAim_r, loc=<0 0 0>, rot=<0 0 0 1>}
 base_dyn_eyeAim_l {parent=pack, loc=<-2.7863 -7.6033 8.6345>, rot=<-0.4373 0.4998 0.4404 -0.6041>}
 dyn_eyeAim_l {parent=base_dyn_eyeAim_l, loc=<0 0 0>, rot=<0 0 0 1>}
 base_dyn_eye_r {parent=pack, loc=<3.4427 -9.4489 8.9674>, rot=<-0.4933 0.5095 0.3909 -0.5867>}
 dyn_eye_r {parent=base_dyn_eye_r, loc=<0 0 0>, rot=<0 0 0 1>}
 base_dyn_eye_l {parent=pack, loc=<-2.55 -9.6417 9.0943>, rot=<-0.4581 0.5149 0.4227 -0.5885>}
 dyn_eye_l {parent=base_dyn_eye_l, loc=<0 0 0>, rot=<0 0 0 1>}
]
