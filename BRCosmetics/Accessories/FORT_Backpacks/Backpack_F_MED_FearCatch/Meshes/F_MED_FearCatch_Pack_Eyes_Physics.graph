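# Physics asset for the eyes mesh of the FearCatch backpack cosmetic, giving its two eye-aim bones simulated bodies.
# A large capsule on the pack bone anchors small capsules on the left and right dyn_eyeAim bones.
# Each eye is constrained to the pack with narrow swing limits and little or no twist, with the left eye allowed a wider swing range than the right.
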PhysicsAsset /BRCosmetics/Accessories/FORT_Backpacks/Backpack_F_MED_FearCatch/Meshes/F_MED_FearCatch_Pack_Eyes_Physics
bodies {
 dyn_eyeAim_l [{shapeType=Capsule, length=1, radius=1, shapeName=eye_aim_l_capsule, center=<2.3164 -0.0176 0.0067>, rotation=<rot 19.9999 -0.0011 -90.0001>}]
 dyn_eyeAim_r [{shapeType=Capsule, length=1, radius=1, shapeName=eye_aim_l_capsule, center=<2.2422 0 0>, rotation=<rot 0 -0.0002 -90>}]
 pack [{shapeType=Capsule, length=18.2055, radius=10.3105, center=<-0.0056 -5.6074 2.8554>, rotation=<rot 0 89.911 -89.9053>}]
}
constraints [
 {bone1Name=dyn_eyeAim_l, bone2Name=pack, swing1LimitDegrees=20, swing2LimitDegrees=25, twistLimitDegrees=0}
 {bone1Name=dyn_eyeAim_r, bone2Name=pack, swing1LimitDegrees=15, swing2LimitDegrees=15, twistLimitDegrees=1}
]
