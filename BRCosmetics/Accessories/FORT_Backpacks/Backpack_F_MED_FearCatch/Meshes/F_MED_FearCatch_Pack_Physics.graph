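# Physics asset for the F_MED_FearCatch backpack, defining collision bodies and joints for the pack and its four dangling or jiggling leg chains, two per side.
# Each chain runs from a base bone locked angularly to the pack, through two joints with narrow swing and tight twist limits.
# All bodies are capsules: a large one for the pack, tiny ones on base and middle leg bones, and larger offset ones on the end bones.
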PhysicsAsset /BRCosmetics/Accessories/FORT_Backpacks/Backpack_F_MED_FearCatch/Meshes/F_MED_FearCatch_Pack_Physics
bodies {
 base_dyn_leg_l_00 [{shapeType=Capsule, length=0.505, radius=0.505, shapeName=base_dyn_leg_l_00_capsule, rotation=<rot 0 -90 0>}]
 base_dyn_leg_l_01 [{shapeType=Capsule, length=0.505, radius=0.505, shapeName=base_dyn_leg_l_01_capsule, rotation=<rot 0 -90 0>}]
 base_dyn_leg_r_00 [{shapeType=Capsule, length=0.505, radius=0.505, shapeName=base_dyn_leg_r_00_capsule, rotation=<rot 0 -90 0>}]
 base_dyn_leg_r_01 [{shapeType=Capsule, length=0.505, radius=0.505, shapeName=base_dyn_leg_r_01_capsule, rotation=<rot 0 -90 0>}]
 dyn_leg_l_00 [{shapeType=Capsule, length=0.505, radius=0.505, shapeName=dyn_leg_l_00_capsule, rotation=<rot 0 -90 0>}]
 dyn_leg_l_00_01 [{
   shapeType=Capsule
   length=3.0346
   radius=2.5784
   shapeName=dyn_leg_l_00_01_capsule
   center=<3.3686 -0.4862 0.1706>
   rotation=<rot 0 83.6912 -75.6706>
  }]
 dyn_leg_l_01 [{shapeType=Capsule, length=0.505, radius=0.505, shapeName=dyn_leg_l_01_capsule, rotation=<rot 0 -90 0>}]
 dyn_leg_l_01_01 [{
   shapeType=Capsule
   length=2.9572
   radius=2.1049
   shapeName=dyn_leg_l_01_01_capsule
   center=<2.5326 -0.2992 -0.2019>
   rotation=<rot -8.0981 78.7888 -80.2722>
  }]
 dyn_leg_r_00 [{shapeType=Capsule, length=0.505, radius=0.505, shapeName=dyn_leg_r_00_capsule, rotation=<rot -0.0002 -90.0001 -5.0001>}]
 dyn_leg_r_00_01 [{
   shapeType=Capsule
   length=3.0346
   radius=2.0782
   shapeName=dyn_leg_r_00_01_capsule
   center=<3.372 -0.4858 -0.0829>
   rotation=<rot -0.0003 -96.2645 -77.1537>
  }]
 dyn_leg_r_01 [{shapeType=Capsule, length=0.505, radius=0.505, shapeName=dyn_leg_r_01_capsule, rotation=<rot 0 -90 0>}]
 dyn_leg_r_01_01 [{
   shapeType=Capsule
   length=2.9572
   radius=2.1046
   shapeName=dyn_leg_r_01_01_capsule
   center=<2.5399 -0.3003 0.0087>
   rotation=<rot -2.552 -97.1433 -75.3601>
  }]
 pack [{shapeType=Capsule, length=14.1386, radius=10.3103, shapeName=pack_capsule, center=<0.0126 -5.6799 2.8555>, rotation=<rot 0 89.7795 -89.6281>}]
}
constraints [
 {
  bone1Name=base_dyn_leg_l_00
  bone2Name=pack
  swing1LimitDegrees=45
  swing1Motion=Locked
  swing2LimitDegrees=45
  swing2Motion=Locked
  twistLimitDegrees=45
  twistMotion=Locked
 }
 {bone1Name=dyn_leg_l_00, bone2Name=base_dyn_leg_l_00, swing1LimitDegrees=15, swing2LimitDegrees=15, twistLimitDegrees=5}
 {bone1Name=dyn_leg_l_00_01, bone2Name=dyn_leg_l_00, swing1LimitDegrees=15, swing2LimitDegrees=20, twistLimitDegrees=5}
 {
  bone1Name=base_dyn_leg_l_01
  bone2Name=pack
  swing1LimitDegrees=45
  swing1Motion=Locked
  swing2LimitDegrees=45
  swing2Motion=Locked
  twistLimitDegrees=45
  twistMotion=Locked
 }
 {bone1Name=dyn_leg_l_01, bone2Name=base_dyn_leg_l_01, swing1LimitDegrees=15, swing2LimitDegrees=15, twistLimitDegrees=5}
 {bone1Name=dyn_leg_l_01_01, bone2Name=dyn_leg_l_01, swing1LimitDegrees=10, swing2LimitDegrees=18, twistLimitDegrees=5}
 {
  bone1Name=base_dyn_leg_r_00
  bone2Name=pack
  swing1LimitDegrees=45
  swing1Motion=Locked
  swing2LimitDegrees=45
  swing2Motion=Locked
  twistLimitDegrees=45
  twistMotion=Locked
 }
 {bone1Name=dyn_leg_r_00, bone2Name=base_dyn_leg_r_00, swing1LimitDegrees=15, swing2LimitDegrees=15, twistLimitDegrees=5}
 {bone1Name=dyn_leg_r_00_01, bone2Name=dyn_leg_r_00, swing1LimitDegrees=15, swing2LimitDegrees=20, twistLimitDegrees=5}
 {
  bone1Name=base_dyn_leg_r_01
  bone2Name=pack
  swing1LimitDegrees=45
  swing1Motion=Locked
  swing2LimitDegrees=45
  swing2Motion=Locked
  twistLimitDegrees=45
  twistMotion=Locked
 }
 {bone1Name=dyn_leg_r_01, bone2Name=base_dyn_leg_r_01, swing1LimitDegrees=10, swing2LimitDegrees=10, twistLimitDegrees=5}
 {bone1Name=dyn_leg_r_01_01, bone2Name=dyn_leg_r_01, swing1LimitDegrees=15, swing2LimitDegrees=15, twistLimitDegrees=5}
]
