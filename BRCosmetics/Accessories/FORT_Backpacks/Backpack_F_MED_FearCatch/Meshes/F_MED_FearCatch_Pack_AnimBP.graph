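# Animation Blueprint for the FearCatch backpack cosmetic, adding secondary physics motion to its leg and eye parts.
# Four chained spring-driven Anim Dynamics nodes simulate left and right leg bones, feeding two RigidBody nodes using pack and eye physics assets.
# An additive pack-bone offset then precedes a final Control Rig.
# The event graph stores a parameter-asset gravity override, a skydiving flag, and front-end or in-zone simulation-space settings chosen on customization load.
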AnimBlueprint /BRCosmetics/Accessories/FORT_Backpacks/Backpack_F_MED_FearCatch/Meshes/F_MED_FearCatch_Pack_AnimBP
target_skeleton=&/F_MED_FearCatch_Pack_Skeleton
graphs [
 AnimGraph {nodes [
   n1 = Anim Dynamics - Bone: dyn_leg_r_02!(ComponentPose=Local To Component(LocalPose=Input Pose()), Alpha=1, GravityOverride:Vector="0,0,-980", ExternalForce="0,0,0") -> (Pose) {
    angularDampingOverride=0.85
    angularSpringConstant=9000
    bAngularSpring=true
    bOverrideAngularDamping=true
    bUseGravityOverride=true
    gravityScale=0.35
    simulationSpace=RootRelative
    boundBone {boneName=dyn_leg_r_02}
    componentLinearAccScale=<1 1 1>
    componentLinearVelScale=<1 1 1>
    gravityOverride {z=-980}
    physicsBodyDefinitions [{
      boundBone {boneName=dyn_leg_r_02}
      boxExtents=<30 40 40>
      constraintSetup {angularLimitsMax=<20 0 10>, angularLimitsMin=<-20 -25 -10>, angularTarget=<1 0 0>}
      localJointOffset=<0 -1 3>
     }]
   }
   n2 = Anim Dynamics - Bone: dyn_leg_r_03!(ComponentPose=@n1, Alpha=1, GravityOverride:Vector="0,0,-980", ExternalForce="0,0,0") -> (Pose) {
    angularDampingOverride=0.9
    angularSpringConstant=15000
    bAngularSpring=true
    bOverrideAngularDamping=true
    bUseGravityOverride=true
    gravityScale=0.35
    simulationSpace=RootRelative
    boundBone {boneName=dyn_leg_r_03}
    componentLinearAccScale=<1 1 1>
    componentLinearVelScale=<1 1 1>
    gravityOverride {z=-980}
    physicsBodyDefinitions [{
      boundBone {boneName=dyn_leg_r_03}
      boxExtents=<25 35 35>
      constraintSetup {angularLimitsMax=<20 0 15>, angularLimitsMin=<-20 -25 -15>, angularTarget=<1 0 0>}
      localJointOffset=<1 -0.5 2>
     }]
   }
   n3 = Anim Dynamics - Bone: dyn_leg_l_02!(ComponentPose=@n2, Alpha=1, GravityOverride:Vector="0,0,-980", ExternalForce="0,0,0") -> (Pose) {
    angularDampingOverride=0.88
    angularSpringConstant=11000
    bAngularSpring=true
    bOverrideAngularDamping=true
    bUseGravityOverride=true
    gravityScale=0.35
    simulationSpace=RootRelative
    boundBone {boneName=dyn_leg_l_02}
    componentLinearAccScale=<1 1 1>
    componentLinearVelScale=<1 1 1>
    gravityOverride {z=-980}
    physicsBodyDefinitions [{
      boundBone {boneName=dyn_leg_l_02}
      boxExtents=<35 45 45>
      constraintSetup {angularLimitsMax=<20 25 12>, angularLimitsMin=<-20 0 -12>, angularTarget=<1 0 0>}
      localJointOffset=<0 -1 -3>
     }]
   }
   n4 = Anim Dynamics - Bone: dyn_leg_l_03!(ComponentPose=@n3, Alpha=1, GravityOverride:Vector="0,0,-980", ExternalForce="0,0,0") -> (Pose) {
    angularDampingOverride=0.99
    angularSpringConstant=12000
    bAngularSpring=true
    bOverrideAngularDamping=true
    bUseGravityOverride=true
    gravityScale=0.35
    simulationSpace=RootRelative
    boundBone {boneName=dyn_leg_l_03}
    componentLinearAccScale=<1 1 1>
    componentLinearVelScale=<1 1 1>
    gravityOverride {z=-980}
    physicsBodyDefinitions [{
      boundBone {boneName=dyn_leg_l_03}
      boxExtents=<20 30 30>
      constraintSetup {angularLimitsMax=<20 25 18>, angularLimitsMin=<-20 0 -18>, angularTarget=<1 0 0>}
      localJointOffset=<1 -0.5 -2>
     }]
   }
   n5 = RigidBody!(
    ComponentPose=@n4
    Alpha
    OverrideWorldGravity:Vector="0,0,0"
    ExternalForce
    SimSpaceSettings:Sim Space Settings Structure=(WorldAlpha=0.5,VelocityScaleZ=1,DampingAlpha=0,MaxLinearVelocity=700,MaxAngularVelocity=8,MaxLinearAcceleration=3500,MaxAngularAcceleration=40,ExternalLinearDragV=(X=0.3,Y=0.3,Z=0.3),ExternalLinearVelocity=(X=0,Y=0,Z=0),ExternalAngularVelocity=(X=0,Y=0,Z=0))
   ) -> (Pose) {
    bOverrideWorldGravity=true
    overridePhysicsAsset=&/F_MED_FearCatch_Pack_Physics
    simSpaceSettings {
     dampingAlpha=0
     maxAngularAcceleration=40
     maxAngularVelocity=8
     maxLinearAcceleration=3500
     maxLinearVelocity=700
     worldAlpha=0.5
     externalLinearDragV=<0.3 0.3 0.3>
    }
   }
   n6 = RigidBody!(ComponentPose=@n5, Alpha, ExternalForce) -> (Pose) {
    bOverrideWorldGravity=true
    overridePhysicsAsset=&/F_MED_FearCatch_Pack_Eyes_Physics
    simSpaceSettings {
     dampingAlpha=0
     maxAngularAcceleration=40
     maxAngularVelocity=8
     maxLinearAcceleration=3500
     maxLinearVelocity=700
     worldAlpha=0.15
     externalLinearDragV=<0.3 0.3 0.3>
    }
   }
   n7 = Control Rig(Source=Component To Local(ComponentPose="Transform (Modify) Bone - Bone: pack"(ComponentPose=@n6, Translation="0,-2,0", Rotation="0,0,0") {translationMode=BMM_Additive, translationSpace=BCS_BoneSpace, boneToModify {boneName=pack}, translation {y=-2}})) {controlRigAssetReference {blueprintRigClass=&/F_MED_FearCatch_Pack_CtrlRig.F_MED_FearCatch_Pack_CtrlRig_C}}
   Output Pose(Result=@n7) {layerGroup=None}
  ]}
 EventGraph {nodes [
   Event BlueprintUpdateAnimation()
   > n1 = "Gravity Override State Function (Parameter Asset) (Nativized)"(
    CharacterPartAnimInstance=Self-Reference()
    GameTimeInSeconds=0
    Parameters=&/F_MED_FearCatch_Pack_Parameters
   )
   > Set GravityOverride(GravityOverride=@n1)
   Event OnBeginSkydiving()
   > Set isSkydive(isSkydive:double=1) -> (Output_Get:double)
   Event OnEndSkydiving()
   > Set isSkydive(isSkydive=0)
   Event OnCharacterCustomizationLoaded()
   > n2 = Branch(Condition=Get bIsInFrontEnd())
   > Set Sim Space Settings(Sim Space Settings:Sim Space Settings Structure=Get Sim Space Settings Front End() -> (Sim Space Settings Front End:Sim Space Settings Structure)) -> (Output_Get:Sim Space Settings Structure)
   Set Sim Space Settings(execute=@n2.else, Sim Space Settings=Get Sim Space Settings in Zone() -> (Sim Space Settings in Zone:Sim Space Settings Structure))
   TryGetPawnOwner()
  ]}
]
variables [
 GravityOverride {type=Vector}
 isSkydive {type=number}
 Sim Space Settings {
  type=SimSpaceSettings
  default {
   dampingAlpha=1
   maxAngularAcceleration=10000
   maxAngularVelocity=10000
   maxLinearAcceleration=10000
   maxLinearVelocity=10000
   velocityScaleZ=1
   worldAlpha=0
   externalLinearDragV=<0 0 0>
  }
 }
 Sim Space Settings Front End {
  type=SimSpaceSettings
  default {
   dampingAlpha=0
   maxAngularAcceleration=40
   maxAngularVelocity=8
   maxLinearAcceleration=3500
   maxLinearVelocity=700
   velocityScaleZ=1
   worldAlpha=1
   externalLinearDragV=<0.3 0.3 0.3>
  }
 }
 Sim Space Settings in Zone {
  type=SimSpaceSettings
  default {
   dampingAlpha=0
   maxAngularAcceleration=40
   maxAngularVelocity=8
   maxLinearAcceleration=3500
   maxLinearVelocity=700
   velocityScaleZ=1
   worldAlpha=0.5
   externalLinearDragV=<0.3 0.3 0.3>
  }
 }
]
